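# Unlit, two-sided translucent ribbon and beam trail material for the Fire Dance emote, drawing stylized fire that gives way to dark smoke.
# Two cartoon flame samples form fire and smoke shapes, with UVs distorted by panning noise that strengthens along U.
# A U gradient minus panning noise, run through Ceil, creates hard-edged dissolves scaled by a dynamic parameter.
# Fire colors get HDR remapping and blend over smoke; the stepped mask times particle alpha drives opacity.
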Material /Game/Effects/Fort_Effects/Emotes/FireDance/M_Fire_Dance_Trail
graph {
 properties {
  bUsedWithBeamTrails=true
  bUsedWithNiagaraRibbons=true
  blendMode=BLEND_Translucent
  pixelDepthOffsetMode=PDOM_Legacy
  shadingModel=MSM_Unlit
  twoSided=true
  usedShadingModels=Unlit
 }
 nodes [
  n1 = VectorParameter() {parameterName=UV Scale And Offset Smoke, defaultValue=<rgba 0.45 0.55 0.2 0.225>}
  n2 = Multiply(A=TextureCoordinate(), B=ComponentMask(None=@n1.RGB) {g=true, r=true})
  n3 = AppendVector(A=@n1.B, B=@n1.B)
  n4 = TextureCoordinate() {uTiling=0.45, vTiling=0.3}
  n5 = AppendVector(A=ScalarParameter() {defaultValue=1, parameterName=Noise X Scale}, B=ScalarParameter() {defaultValue=1, parameterName=Noise Y Scale})
  n6 = ScalarParameter() {defaultValue=1, parameterName=Noise Pan Speed Scale}
  n7 = Constant2Vector() {g=-0.2, r=-1.9}
  n8 = TextureSample(UVs=Panner(Coordinate=Multiply(A=@n4, B=@n5), Speed=Multiply(A=@n6, B=@n7)) {speedX=-1.9, speedY=-0.2}) {
   samplerSource=SSM_Wrap_WorldGroupSettings
   samplerType=SAMPLERTYPE_LinearColor
   texture=&T_NoisePack_LinearColor
  }
  n9 = TextureCoordinate()
  n10 = Multiply(A=Multiply(B=ComponentMask(None=@n8.RGB) {g=true, r=true}) {constA=0.65}, B=Power(Base=ComponentMask(None=@n9) {r=true}))
  n11 = TextureSample(UVs=Add(A=@n10, B=Add(A=@n2, B=@n3))) {samplerSource=SSM_Clamp_WorldGroupSettings, texture=/Game/Effects/Fort_Effects/Textures/Fire/T_CartoonFlame}
  n12 = LinearInterpolate(
   A=VectorParameter() {parameterName=Smoke Primary, defaultValue=<rgba 0.0052 0.0052 0.0052 1>}.RGB
   B=VectorParameter() {parameterName=Smoke Secondary, defaultValue=<rgba 0.0885 0.0885 0.0885 1>}.RGB
   Alpha=Multiply(A=@n11.R, B=@n11.R)
  )
  n13 = OneMinus(None=ComponentMask(None=TextureCoordinate()) {r=true})
  n14 = VectorParameter() {parameterName=UV Scale And Offset, defaultValue=<rgba 1.65 0.33 0 0.33>}
  n15 = Multiply(A=TextureCoordinate(), B=ComponentMask(None=@n14.RGB) {g=true, r=true})
  n16 = AppendVector(A=@n14.B, B=@n14.B)
  n17 = TextureSample(UVs=Add(A=Add(A=@n15, B=@n16), B=@n10)) {samplerSource=SSM_Clamp_WorldGroupSettings, texture=&T_CartoonFlame}
  n18 = LinearInterpolate(
   A=VectorParameter() {parameterName=Fire, defaultValue=<rgba 3 0.2043 0 1>}.RGB
   B=VectorParameter() {parameterName=FireBright, defaultValue=<rgba 12 0.7776 0 1>}.RGB
   Alpha=Saturate(None=Add(A=@n13, B=@n17.R))
  )
  n19 = Power(Base=@n13) {constExponent=10}
  n20 = Saturate(None=@n18)
  n21 = Multiply(
   A=@n13
   B=DynamicParameter() {defaultValue=<rgba 1.1 1 1 1>, paramNames [Fire Dissolve Multiplier, Param2, Param3, Param4]}.Fire Dissolve Multiplier
  )
  n22 = ScalarParameter() {defaultValue=4, parameterName=Fire Dissolve Power}
  n23 = TextureCoordinate() {uTiling=2, vTiling=0.5}
  n24 = AppendVector(A=ScalarParameter() {defaultValue=1, parameterName=Noise X Scale}, B=ScalarParameter() {defaultValue=1, parameterName=Noise Y Scale})
  n25 = ScalarParameter() {defaultValue=1, parameterName=Noise Pan Speed Scale}
  n26 = Constant2Vector() {g=0.4, r=-2.25}
  n27 = TextureSample(UVs=Panner(Coordinate=Multiply(A=@n23, B=@n24), Speed=Multiply(A=@n25, B=@n26)) {speedX=-2.25, speedY=0.4}) {
   samplerSource=SSM_Wrap_WorldGroupSettings
   samplerType=SAMPLERTYPE_LinearColor
   texture=&T_NoisePack_LinearColor
  }
  n28 = Ceil(None=Subtract(A=Power(Base=@n21, Exp=@n22) {constExponent=3.4}, B=@n27.G))
  MaterialFunctionCall(Emissive=LinearInterpolate(
    A=LinearInterpolate(A=@n12, B=@n18, Alpha=@n19)
    B=MaterialFunctionCall(LDR Input=@n20, HDR Tint=@n18) {materialFunction=&DeriveHDRfromLDR}.RemappedRange
    Alpha=Multiply(A=@n28, B=@n17.R)
   )) {materialFunction=&MF_FX_EmissiveAdjustments}
  n29 = TextureCoordinate()
  n30 = Constant2Vector() {g=0.5, r=2.125}
  n31 = Add(A=@n17.R, B=Multiply(A=@n11.R, B=SphereMask(A=@n29, B=@n30) {attenuationRadius=2, hardnessPercent=80}))
  n32 = Ceil(None=Subtract(A=@n21, B=@n27.G))
  Multiply(A=ParticleColor().A, B=Ceil(None=Multiply(A=@n31, B=@n32)))
 ]
}
