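# Niagara fire burst effect for the FireDance emote, built from six sprite emitters that spawn instantaneous bursts, with no user parameters.
# A brief flickering Airglow and flipbook-animated SmokeySwirls accompany GPU-simulated ash and embers.
# EmbersCenterFast sprays short-lived sparks radially, while EmbersLeft and EmbersRight launch mirrored narrow cones slowed by drag, gravity and curl noise.
# Embers stretch with speed, and curves drive color and fades across all layers.
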NiagaraSystem /Game/Effects/Fort_Effects/Emotes/FireDance/NS_Firedance_Fire_Burst
dependencies {
 usedDataInterfaces [
  &NiagaraDataInterfaceCurve
  &NiagaraDataInterfaceVectorCurve
  &NiagaraDataInterfaceSpriteRendererInfo
 ]
 usedDynamicInputs [&MakeVector2D]
 usedModules [&SystemState]
}
emitters {
 Airglow {
  emitterUpdateScript [
   &EmitterState(Life Cycle Mode=::NewEnumerator1, Loop Behavior=::NewEnumerator1, Loop Duration=0.1, UseLoopDelay=true)
   &SpawnBurst_Instantaneous(Use Spawn Probability=false, UseLoopCountLimit=false)
  ]
  particleSpawnScript [
   &InitializeParticle(Lifetime=0.2, Sprite Size=<75 75>, Sprite Size Mode=::NewEnumerator4, UsePositionOffset=false, Write Material Random=true)
   &Color(Color=&MakeLinearColorFromVectorAndFloat("Vector (RGB)"=<4 1.369 0.11>, "Float (Alpha)"=&ClampFloat(Float=0.2)))
  ]
  particleUpdateScript [
   &ParticleState()
   &ScaleSpriteSize(
    Scale Sprite Size Mode=::NewEnumerator0
    Uniform Scale Factor=&RandomRangeFloat(Minimum=1.25, Maximum=1.5, Evaluation Type=::NewEnumerator1)
   ) {module=ScaleSpriteSize001}
   &ScaleSpriteSize(
    Scale Sprite Size Mode=::NewEnumerator0
    Uniform Scale Factor=&Waveform("Waveform [1]"=::NewEnumerator9, "[1] Frequency"=5, "[1] Amplitude Min/Max"=<0.8 1.3>)
   )
   &ScaleColor(
    Scale Alpha=&FloatFromCurve(FloatCurve=DataInterface() {propertyValues {Curve {keys [
         <at 0 0.75 RCIM_Cubic>
         {arriveTangent=0.1087, interpMode=RCIM_Cubic, leaveTangent=0.1087, time=0.054, value=1.0088}
         {arriveTangent=-0.4216, interpMode=RCIM_Cubic, leaveTangent=-0.4216, time=0.1549, value=0.7668}
         {arriveTangent=-1.0676, interpMode=RCIM_Cubic, leaveTangent=-1.0676, time=0.2465, value=0.9277}
         {arriveTangent=-0.5051, interpMode=RCIM_Cubic, leaveTangent=-0.5051, time=0.3607, value=0.5471}
         {arriveTangent=-2.0548, interpMode=RCIM_Cubic, leaveTangent=-2.0548, time=0.4457, value=0.827}
         {arriveTangent=-1.8132, interpMode=RCIM_Cubic, leaveTangent=-1.8132, time=0.5437, value=0.1709}
         {arriveTangent=0.3338, interpMode=RCIM_Cubic, leaveTangent=0.3338, time=0.6399, value=0.4749}
         {arriveTangent=2.3758, interpMode=RCIM_Cubic, leaveTangent=2.3758, time=0.7116, value=0.227}
         {arriveTangent=-0.3518, interpMode=RCIM_Cubic, leaveTangent=-0.3518, time=0.7694, value=0.7824}
         {arriveTangent=-3.3922, interpMode=RCIM_Cubic, leaveTangent=-3.3922, time=0.9094, value=0.1574}
         <at 1 0 RCIM_Cubic>
        ]}}})
    ScaleA=true
    ScaleRGB=true
   )
   &DynamicMaterialParameters(
    Index 0 Param 2=1
    Index 0 Param 3=1
    Index 0 Param 4=1
    Module.Param0WriteEnabled=true
    Module.Param1WriteEnabled=true
    Module.Param2WriteEnabled=true
    Module.Param3WriteEnabled=true
   )
  ]
  rendererClasses [&NiagaraSpriteRendererProperties]
  renderers [{rendererClass=&NiagaraSpriteRendererProperties}]
 }
 Ash {
  simTarget=GPUComputeSim
  emitterUpdateScript [
   &EmitterState()
   &SpawnBurst_Instantaneous(
    Spawn Count=&RandomRangeInt(Minimum=11, Maximum=31)
    Use Spawn Probability=false
    UseLoopCountLimit=false
   )
  ]
  particleSpawnScript [
   /SetVariables(Particles.SpriteRotationRate=&RandomRangeFloat(Minimum=-1.5, Maximum=1.5)) {isSetParameters=true}
   &InitializeParticle(
    Lifetime Max=3
    Lifetime Min=0.5
    Lifetime Mode=::NewEnumerator1
    Sprite Rotation Angle=&RandomRangeFloat(Minimum=0, Maximum=1)
    Sprite Rotation Mode=::NewEnumerator1
    Sprite Size Max=<2.8 6>
    Sprite Size Min=<4.5 6>
    Sprite Size Mode=::NewEnumerator2
    UsePositionOffset=false
    Write Material Random=true
   )
   &Color(Color=&MakeLinearColorFromVectorAndFloat("Vector (RGB)"=<20 2 2>, "Float (Alpha)"=&ClampFloat(Float=1)))
   &ShapeLocation(
    Cylinder Height=0.01
    Cylinder Radius=10
    MaskSpawnGroup=false
    Override Seed=false
    Rotation Axis=<0 1 0>
    Rotation Mode=::NewEnumerator1
    Shape Primitive=::NewEnumerator1
    Surface Only=true
    Use Endcaps In Surface Only Mode=false
    UseOwnerScale=true
   )
   &AddVelocity(Velocity=&Multiply_VectorByFloat(
     Vector=&Multiply_Vector(A=$Particles.ShapeLocation.ShapeVector, B=<1 0 1>)
     Float=&RandomRangeFloat(Minimum=15, Maximum=50)
    ))
  ]
  particleUpdateScript [
   &ParticleState()
   &Drag(Use Linear Drag=true, Use Rotational Drag=true)
   &CurlNoiseForce(Noise Frequency=25, Noise Strength=1000, Pan Noise=false)
   &AccelerationForce(Acceleration=<0 0 -500>)
   &SolveForcesAndVelocity(Clamp Velocity=false, Limit Acceleration=false)
   &ScaleSpriteSize(
    Uniform Curve Scale=3
    Uniform Curve Sprite Scale=DataInterface() {propertyValues {Curve {keys [
        {arriveTangent=9.0725, interpMode=RCIM_Cubic, leaveTangent=9.0725, tangentMode=RCTM_User, time=0, value=0}
        {interpMode=RCIM_Cubic, tangentMode=RCTM_User, time=0.025, value=0.9847}
        {arriveTangent=-1.0099, interpMode=RCIM_Cubic, leaveTangent=-1.0099, time=0.06, value=0.33}
        {arriveTangent=-0.0405, interpMode=RCIM_Cubic, leaveTangent=-0.0405, tangentMode=RCTM_User, time=1, value=0}
       ]}}}
   )
   &SpriteRotationRate(Rotation Rate=&AngleConversion(
     Angle:NiagaraFloat=$Particles.SpriteRotationRate
     Angle Input:ENiagara_AngleInput=::NewEnumerator1
     Angle Output:ENiagara_AngleInput=::NewEnumerator0
    ))
   &ScaleColor(
    Scale RGB=&VectorFromCurve(VectorCurve=DataInterface() {propertyValues {
       XCurve {keys [<at 0 500 RCIM_Cubic>, {arriveTangent=-500, interpMode=RCIM_Cubic, leaveTangent=-500, time=0.68, value=10}, <at 1 0 RCIM_Cubic>]}
       YCurve {keys [
         <at 0 154.533 RCIM_Cubic>
         {arriveTangent=-154.533, interpMode=RCIM_Cubic, leaveTangent=-154.533, time=0.68, value=0.426}
         <at 1 0 RCIM_Cubic>
        ]}
       ZCurve {keys [<at 0 49.0353 RCIM_Cubic>, {arriveTangent=-49.0353, interpMode=RCIM_Cubic, leaveTangent=-49.0353, time=0.68, value=0}, <at 1 0 RCIM_Cubic>]}
      }})
    ScaleA=true
    ScaleRGB=true
   )
  ]
  rendererClasses [&NiagaraSpriteRendererProperties]
  renderers [{rendererClass=&NiagaraSpriteRendererProperties}]
 }
 EmbersCenterFast {
  simTarget=GPUComputeSim
  emitterUpdateScript [
   &EmitterState(Life Cycle Mode=::NewEnumerator1, Loop Behavior=::NewEnumerator1, Loop Delay=0.025, Loop Duration=0.5, UseLoopDelay=true)
   &SpawnBurst_Instantaneous(
    Spawn Count=&RandomRangeInt(Minimum=23, Maximum=71)
    Use Spawn Probability=false
    UseLoopCountLimit=false
   )
  ]
  particleSpawnScript [
   &InitializeParticle(
    Color=<rgba 4 1.2 0.1 1>
    Lifetime=&RandomRangeFloat(Minimum=0.065, Maximum=0.1)
    Sprite Size=&RandomRangeVector2D(Minimum=<3 4>, Maximum=<4 6>)
    Sprite Size Mode=::NewEnumerator4
    UsePositionOffset=false
    Write Material Random=true
   )
   &ShapeLocation(Hemisphere Distribution=<1 0.5>, MaskSpawnGroup=false, Override Seed=false, Sphere Radius=5, UseOwnerScale=true)
   &AddVelocity(Velocity=&Multiply_VectorByFloat(Vector=$Particles.ShapeLocation.ShapeVector, Float=&RandomRangeFloat(Minimum=100, Maximum=350)))
  ]
  particleUpdateScript [
   &ParticleState()
   &Drag(Drag=3, Use Linear Drag=true, Use Rotational Drag=true)
   &AccelerationForce(Acceleration=<0 0 -350>)
   &SolveForcesAndVelocity(Clamp Velocity=false, Limit Acceleration=false)
   &ScaleColor(
    Scale Alpha=&FloatFromCurve(FloatCurve=DataInterface() {propertyValues {Curve {keys [{arriveTangent=11.7282, interpMode=RCIM_Cubic, time=0, value=0}, <at 0.2 1 RCIM_Cubic>, <at 1 0>]}}})
    Scale RGB=&VectorFromCurve(VectorCurve=DataInterface() {propertyValues {XCurve {keys [<at 0 4>,<at 1 1>]}, YCurve {keys [<at 0 1.3836>,<at 1 0.0572>]}, ZCurve {keys [<at 0 0>,<at 1 0>]}}})
    ScaleA=true
    ScaleRGB=true
   )
   &ScaleSpriteSize(Uniform Curve Sprite Scale=DataInterface() {propertyValues {Curve {keys [<at 0.07 1.8>, {arriveTangent=-0.0405, interpMode=RCIM_Cubic, leaveTangent=-0.0405, tangentMode=RCTM_User, time=1, value=0}]}}})
   &ScaleSpriteSizeBySpeed(Max Scale Factor=<1 4>, Source Velocity:Vector3f=&Multiply_Vector(A=<0 0.01 1>, B=$Particles.Velocity))
  ]
  rendererClasses [&NiagaraSpriteRendererProperties]
  renderers [{rendererClass=&NiagaraSpriteRendererProperties}]
 }
 EmbersLeft {
  simTarget=GPUComputeSim
  emitterUpdateScript [
   &EmitterState(Life Cycle Mode=::NewEnumerator1, Loop Behavior=::NewEnumerator1, Loop Delay=0.028, Loop Duration=0.5, UseLoopDelay=true)
   &SpawnBurst_Instantaneous(
    Spawn Count=&RandomRangeInt(Minimum=11, Maximum=31)
    Use Spawn Probability=false
    UseLoopCountLimit=false
   )
  ]
  particleSpawnScript [
   &InitializeParticle(
    Lifetime=&RandomRangeFloat(Minimum=0.2, Maximum=1)
    Sprite Size=&RandomRangeVector2D(Minimum=<2 3>, Maximum=<3 5>)
    Sprite Size Mode=::NewEnumerator4
    UsePositionOffset=false
    Write Material Random=true
   )
   &Color(Color=&MakeLinearColorFromVectorAndFloat("Vector (RGB)"=<4 1.2 0.1>, "Float (Alpha)"=&ClampFloat(Float=1)))
   &AddVelocity(Velocity=&RandomRangeVector(Minimum=<0 0 -80>, Maximum=<50 0 -150>))
   &AddVelocity(
    Cone Angle=5
    Cone Axis=<0.26 0 1>
    Module.UseConeFalloff=false
    Module.UseVelDistribution=false
    Override Seed=false
    Velocity Mode=::NewEnumerator2
    Velocity Speed=&RandomRangeFloat(Minimum=50, Maximum=800)
   ) {module=AddVelocity001}
  ]
  particleUpdateScript [
   &ParticleState()
   &Drag(
    Drag=&RandomRangeFloat(Minimum=0, Maximum=3, Evaluation Type=::NewEnumerator1)
    Use Linear Drag=true
    Use Rotational Drag=true
   )
   &AccelerationForce(Acceleration=<0 0 -350>)
   &CurlNoiseForce(Noise Frequency=25, Noise Strength=1000, Pan Noise=false)
   &SolveForcesAndVelocity(Clamp Velocity=false, Limit Acceleration=false)
   &ScaleColor(
    Scale Alpha=&FloatFromCurve(FloatCurve=DataInterface() {propertyValues {Curve {keys [{arriveTangent=11.7282, interpMode=RCIM_Cubic, time=0, value=0}, <at 0.2 1 RCIM_Cubic>, <at 1 0>]}}})
    Scale RGB=&VectorFromCurve(VectorCurve=DataInterface() {propertyValues {XCurve {keys [<at 0 4>,<at 1 1>]}, YCurve {keys [<at 0 1.3836>,<at 1 0.0572>]}, ZCurve {keys [<at 0 0>,<at 1 0>]}}})
    ScaleA=true
    ScaleRGB=true
   )
   &ScaleSpriteSize(Uniform Curve Sprite Scale=DataInterface() {propertyValues {Curve {keys [
        {arriveTangent=9.0725, interpMode=RCIM_Cubic, leaveTangent=9.0725, tangentMode=RCTM_User, time=0, value=0}
        {arriveTangent=-1.3009, interpMode=RCIM_Cubic, leaveTangent=-1.3009, tangentMode=RCTM_User, time=0.05, value=0.9847}
        {arriveTangent=-0.0405, interpMode=RCIM_Cubic, leaveTangent=-0.0405, tangentMode=RCTM_User, time=1, value=0}
       ]}}})
   &ScaleSpriteSizeBySpeed(Max Scale Factor=<1 4>, Source Velocity:Vector3f=&Multiply_Vector(A=<0 0.01 1>, B=$Particles.Velocity))
  ]
  rendererClasses [&NiagaraSpriteRendererProperties]
  renderers [{rendererClass=&NiagaraSpriteRendererProperties}]
 }
 EmbersRight {
  simTarget=GPUComputeSim
  emitterUpdateScript [
   &EmitterState(Life Cycle Mode=::NewEnumerator1, Loop Behavior=::NewEnumerator1, Loop Delay=0.028, Loop Duration=0.5, UseLoopDelay=true)
   &SpawnBurst_Instantaneous(
    Spawn Count=&RandomRangeInt(Minimum=11, Maximum=31)
    Use Spawn Probability=false
    UseLoopCountLimit=false
   )
  ]
  particleSpawnScript [
   &InitializeParticle(
    Lifetime=&RandomRangeFloat(Minimum=0.2, Maximum=1)
    Sprite Size=&RandomRangeVector2D(Minimum=<2 3>, Maximum=<3 5>)
    Sprite Size Mode=::NewEnumerator4
    UsePositionOffset=false
    Write Material Random=true
   )
   &Color(Color=&MakeLinearColorFromVectorAndFloat("Vector (RGB)"=<4 1.2 0.1>, "Float (Alpha)"=&ClampFloat(Float=1)))
   &AddVelocity(Velocity=&RandomRangeVector(Minimum=<0 0 -80>, Maximum=<50 0 -150>))
   &AddVelocity(
    Cone Angle=5
    Cone Axis=<-0.26 0 1>
    Module.UseConeFalloff=false
    Module.UseVelDistribution=false
    Override Seed=false
    Velocity Mode=::NewEnumerator2
    Velocity Speed=&RandomRangeFloat(Minimum=50, Maximum=800)
   ) {module=AddVelocity001}
  ]
  particleUpdateScript [
   &ParticleState()
   &ScaleColor(
    Scale Alpha=&FloatFromCurve(FloatCurve=DataInterface() {propertyValues {Curve {keys [{arriveTangent=11.7282, interpMode=RCIM_Cubic, time=0, value=0}, <at 0.2 1 RCIM_Cubic>, <at 1 0>]}}})
    Scale RGB=&VectorFromCurve(VectorCurve=DataInterface() {propertyValues {XCurve {keys [<at 0 4>,<at 1 1>]}, YCurve {keys [<at 0 1.3836>,<at 1 0.0572>]}, ZCurve {keys [<at 0 0>,<at 1 0>]}}})
    ScaleA=true
    ScaleRGB=true
   )
   &Drag(
    Drag=&RandomRangeFloat(Minimum=0, Maximum=3, Evaluation Type=::NewEnumerator1)
    Use Linear Drag=true
    Use Rotational Drag=true
   )
   &AccelerationForce(Acceleration=<0 0 -350>)
   &CurlNoiseForce(Noise Frequency=25, Noise Strength=1000, Pan Noise=false)
   &SolveForcesAndVelocity(Clamp Velocity=false, Limit Acceleration=false)
   &ScaleSpriteSizeBySpeed(Max Scale Factor=<1 4>, Source Velocity:Vector3f=&Multiply_Vector(A=<0 0.01 1>, B=$Particles.Velocity))
   &ScaleSpriteSize(Uniform Curve Sprite Scale=DataInterface() {propertyValues {Curve {keys [
        {arriveTangent=9.0725, interpMode=RCIM_Cubic, leaveTangent=9.0725, tangentMode=RCTM_User, time=0, value=0}
        {arriveTangent=-1.3009, interpMode=RCIM_Cubic, leaveTangent=-1.3009, tangentMode=RCTM_User, time=0.05, value=0.9847}
        {arriveTangent=-0.0405, interpMode=RCIM_Cubic, leaveTangent=-0.0405, tangentMode=RCTM_User, time=1, value=0}
       ]}}})
  ]
  rendererClasses [&NiagaraSpriteRendererProperties]
  renderers [{rendererClass=&NiagaraSpriteRendererProperties}]
 }
 SmokeySwirls {
  emitterUpdateScript [
   &EmitterState(Life Cycle Mode=::NewEnumerator1, Loop Behavior=::NewEnumerator1, Loop Duration=0.2, UseLoopDelay=true)
   &SpawnBurst_Instantaneous(Spawn Count=3, Use Spawn Probability=false, UseLoopCountLimit=false)
  ]
  particleSpawnScript [
   &InitializeParticle(
    Color=<rgba 1 0.5 0.5 1>
    Lifetime=&RandomRangeFloat(Minimum=0.4, Maximum=0.8)
    Sprite Rotation Mode=::NewEnumerator2
    Sprite Size=&Vector2DFromFloat(Value=&RandomRangeFloat(Minimum=180, Maximum=200))
    Sprite Size Mode=::NewEnumerator4
    UsePositionOffset=false
    Write Material Random=true
   )
   &ShapeLocation(
    Cylinder Height=0.01
    Cylinder Radius=10
    Hemicircle Y=true
    MaskSpawnGroup=false
    Offset=<0 -15 0>
    Offset Mode=::NewEnumerator0
    Override Seed=false
    Rotation Angle=0.5
    Rotation Axis=<0 1 0>
    Rotation Mode=::NewEnumerator1
    Shape Primitive=::NewEnumerator1
    Surface Only=true
    UseOwnerScale=true
   )
   &AddVelocity(Velocity=&Multiply_VectorByFloat(
     Vector=&Multiply_Vector(A=$Particles.ShapeLocation.ShapeVector, B=<1 0 1>)
     Float=&RandomRangeFloat(Minimum=15, Maximum=50)
    ))
  ]
  particleUpdateScript [
   &ParticleState()
   &Drag(
    Drag=&RandomRangeFloat(Minimum=3, Maximum=5, Evaluation Type=::NewEnumerator1)
    Use Linear Drag=true
    Use Rotational Drag=false
   )
   &AccelerationForce(Acceleration=&RandomRangeVector(Minimum=<0 0 -3.85>, Maximum=<0 0 -6>))
   &SolveForcesAndVelocity(Clamp Velocity=false, Limit Acceleration=false)
   &SubUVAnimation(
    End Frame Range Override=48
    Start Frame Range Override=4
    SubUV Animation Mode:ENiagara_SubUVLookupModeV2=::NewEnumerator1
    SubUV Curve=DataInterface() {propertyValues {Curve {keys [
        {arriveTangent=4.1785, interpMode=RCIM_Cubic, leaveTangent=4.1785, tangentMode=RCTM_User, time=0, value=0}
        {arriveTangent=1.1179, interpMode=RCIM_Cubic, leaveTangent=1.1179, tangentMode=RCTM_User, time=0.2042, value=0.6228}
        {arriveTangent=0.0224, interpMode=RCIM_Cubic, leaveTangent=0.0224, tangentMode=RCTM_User, time=1, value=1}
       ]}}}
    UseEndFrame=true
    UseStartFrame=true
   )
   &ScaleSpriteSize(Uniform Curve Sprite Scale=DataInterface() {propertyValues {Curve {keys [
        {arriveTangent=4.1785, interpMode=RCIM_Cubic, leaveTangent=4.1785, tangentMode=RCTM_User, time=0, value=0}
        {arriveTangent=1.1179, interpMode=RCIM_Cubic, leaveTangent=1.1179, tangentMode=RCTM_User, time=0.2042, value=0.6228}
        {arriveTangent=0.0224, interpMode=RCIM_Cubic, leaveTangent=0.0224, tangentMode=RCTM_User, time=1, value=1}
       ]}}})
   &ScaleColor(
    Scale Alpha=&FloatFromCurve(FloatCurve=DataInterface() {propertyValues {Curve {keys [
         {arriveTangent=3.8883, interpMode=RCIM_Cubic, time=0, value=0}
         {arriveTangent=3.5736, interpMode=RCIM_Cubic, leaveTangent=3.5736, time=0.3538, value=0.9853}
         {arriveTangent=-1.5248, interpMode=RCIM_Cubic, leaveTangent=-1.5248, time=0.605, value=2.1621}
         {interpMode=RCIM_Cubic, leaveTangent=-2.7493, time=1, value=0}
        ]}}})
    Scale RGB=&VectorFromCurve(VectorCurve=DataInterface() {propertyValues {XCurve {keys [<at 0 0.5>,<at 0.125 0>]}, YCurve {keys [<at 0 0.1>,<at 0.125 0>]}, ZCurve {keys [<at 0 0>,<at 0.125 0>]}}})
    ScaleA=true
    ScaleRGB=true
   )
  ]
  rendererClasses [&NiagaraSpriteRendererProperties]
  renderers [{rendererClass=&NiagaraSpriteRendererProperties}]
 }
}
summary {
 systemName=NS_Firedance_Fire_Burst
 emitters [
  {emitterName=Ash, simTarget=GPUComputeSim, rendererClasses [&NiagaraSpriteRendererProperties]}
  {emitterName=Airglow, rendererClasses [&NiagaraSpriteRendererProperties]}
  {emitterName=EmbersLeft, simTarget=GPUComputeSim, rendererClasses [&NiagaraSpriteRendererProperties]}
  {emitterName=EmbersCenterFast, simTarget=GPUComputeSim, rendererClasses [&NiagaraSpriteRendererProperties]}
  {emitterName=EmbersRight, simTarget=GPUComputeSim, rendererClasses [&NiagaraSpriteRendererProperties]}
  {emitterName=SmokeySwirls, rendererClasses [&NiagaraSpriteRendererProperties]}
 ]
 userVariables []
}
